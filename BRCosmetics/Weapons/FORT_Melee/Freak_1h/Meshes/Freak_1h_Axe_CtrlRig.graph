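# Control Rig for the Freak_1h axe melee weapon that procedurally animates its dangling tongue bone chain.
# On each execution, a single Chain Harmonics unit drives the chain from dyn_tongue_01_Base with a noisy, eased wave.
# Pendulum and reach are disabled, and debug drawing is left on.
# The hierarchy also holds eye, eyelid and tooth bones, which the graph does not directly drive.
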ControlRigBlueprint /BRCosmetics/Weapons/FORT_Melee/Freak_1h/Meshes/Freak_1h_Axe_CtrlRig
graphs [
 RigVMFunctionLibrary {links [], nodes []}
 RigVMModel {nodes [
   n1 = &RigUnit_BeginExecution()
   /Script/ControlRig.RigUnit_ChainHarmonics(
    ChainRoot=dyn_tongue_01_Base
    DrawWorldOffset=(Rotation=(X=0,Y=0,Z=0,W=1),Translation=(X=0,Y=0,Z=0),Scale3D=(X=1,Y=1,Z=1))
    Pendulum=(bEnabled=False,PendulumStiffness=2,PendulumGravity=(X=0,Y=0,Z=0),PendulumBlend=0.75,PendulumDrag=0.98,PendulumMinimum=0,PendulumMaximum=0.1,PendulumEase=Linear,UnwindAxis=(X=0,Y=1,Z=0),UnwindMinimum=0.2,UnwindMaximum=0.05)
    Reach=(bEnabled=False,ReachTarget=(X=0,Y=0,Z=0),ReachAxis=(X=1,Y=0,Z=0),ReachMinimum=0,ReachMaximum=0,ReachEase=SineEaseInOut)
    Speed=(X=0.1,Y=0.25,Z=0.1)
    Wave=(bEnabled=True,WaveFrequency=(X=0.2,Y=0.2,Z=0.2),WaveAmplitude=(X=0.25,Y=2.5,Z=1.5),WaveOffset=(X=1,Y=0,Z=2),WaveNoise=(X=8,Y=8,Z=15),WaveMinimum=0,WaveMaximum=0.75,WaveEase=SineEaseIn)
    WaveCurve=(EditorCurveData=(Keys=((Time=-0.0198,Value=0.4929),(Time=0.9855,Value=0.5049)),PreInfinityExtrap=RCCE_Constant,PostInfinityExtrap=RCCE_Constant,DefaultValue=340282346638528859811704183484516925440))
    bDrawDebug=True
    ExecutePin=@n1.ExecutePin
   ) {is=struct RigVMUnitNode}
  ]}
]
hierarchy [
 b_AR_Root {loc=<0 0 0>, rot=<0 0 0 1>}
 Melee_TwoHanded_Handle {parent=b_AR_Root, loc=<0 0 0>, rot=<0 0 0.7071 -0.7071>}
 dyn_eye_Base_l {parent=Melee_TwoHanded_Handle, loc=<-6.9981 0.9184 52.9671>, rot=<0.1691 -0.6866 0.6866 0.1691>}
 dyn_eye_l {parent=dyn_eye_Base_l, loc=<0 0 0>, rot=<0 0 0 1>}
 dyn_eye_Base_r {parent=Melee_TwoHanded_Handle, loc=<6.784 0.9184 52.9671>, rot=<0.6866 0.1691 0.1691 -0.6866>}
 dyn_eye_r {parent=dyn_eye_Base_r, loc=<0 0 0>, rot=<0 0 0 1>}
 dyn_tongue_01_Base {parent=Melee_TwoHanded_Handle, loc=<0.0171 0.3227 -11.968>, rot=<-0.0706 -0.7036 -0.0645 -0.7042>}
 dyn_tongue_02 {parent=dyn_tongue_01_Base, loc=<3.0876 0 0>, rot=<0.0043 0.0002 0.2394 -0.9709>}
 dyn_tongue_03 {parent=dyn_tongue_02, loc=<4.4055 0 0>, rot=<0.0077 0.0007 -0.2 -0.9798>}
 dyn_tongue_04 {parent=dyn_tongue_03, loc=<3.5673 0 0>, rot=<-0.0078 0.0055 -0.4799 -0.8773>}
 dyn_tongue_05 {parent=dyn_tongue_04, loc=<2.6261 0 0>, rot=<-0.0004 -0.004 -0.5219 -0.853>}
 dyn_tongue_06 {parent=dyn_tongue_05, loc=<2.5131 0 0>, rot=<0.0019 0.001 -0.4114 -0.9115>}
 dyn_eyelidDown_Base_l {parent=Melee_TwoHanded_Handle, loc=<-8.0006 0.5955 52.4479>, rot=<-0.149 -0.6944 -0.6912 0.1337>}
 dyn_eyelidDown_l {parent=dyn_eyelidDown_Base_l, loc=<0 0 0>, rot=<0 0 0 1>}
 dyn_eyelidUp_Base_l {parent=Melee_TwoHanded_Handle, loc=<-8.0006 0.5955 52.4479>, rot=<0.3537 -0.5375 0.6123 0.4594>}
 dyn_eyelidUp_l {parent=dyn_eyelidUp_Base_l, loc=<0 0 0>, rot=<0 0 0 -1>}
 dyn_eyelidDown_Base_r {parent=Melee_TwoHanded_Handle, loc=<7.8649 0.9819 52.5556>, rot=<-0.7006 0.1703 -0.096 -0.6863>}
 dyn_eyelidDown_r {parent=dyn_eyelidDown_Base_r, loc=<0 0 0>, rot=<0 0 0 1>}
 dyn_eyelidUp_Base_r {parent=Melee_TwoHanded_Handle, loc=<7.8649 0.9819 52.5556>, rot=<0.5934 0.4377 0.3845 -0.5553>}
 dyn_eyelidUp_r {parent=dyn_eyelidUp_Base_r, loc=<0 0 0>, rot=<0 0 0 1>}
 dyn_toothJ {parent=Melee_TwoHanded_Handle, loc=<8.3783 -9.7124 44.2168>, rot=<0.5196 -0.6196 -0.266 -0.5247>}
 dyn_toothH {parent=Melee_TwoHanded_Handle, loc=<7.0959 -14.4185 46.3564>, rot=<0.3804 -0.6472 -0.3284 -0.5732>}
 dyn_toothG {parent=Melee_TwoHanded_Handle, loc=<4.1889 -17.5511 46.1729>, rot=<0.287 -0.6407 -0.2562 -0.6644>}
 dyn_toothF {parent=Melee_TwoHanded_Handle, loc=<2.4396 -19.9257 46.2571>, rot=<0.2104 -0.6628 0.055 -0.7165>}
 dyn_toothE {parent=Melee_TwoHanded_Handle, loc=<-0.1946 -19.8716 46.4247>, rot=<0.1663 -0.6873 0.1736 -0.6855>}
 dyn_toothD {parent=Melee_TwoHanded_Handle, loc=<-3.037 -19.0581 46.4182>, rot=<-0.0584 -0.7047 0.1689 -0.6866>}
 dyn_toothC {parent=Melee_TwoHanded_Handle, loc=<-5.2475 -16.1757 45.5874>, rot=<-0.2099 -0.6399 0.322 -0.6654>}
 dyn_toothB {parent=Melee_TwoHanded_Handle, loc=<-7.9768 -9.752 45.6169>, rot=<-0.2999 -0.6087 0.3678 -0.6358>}
 dyn_toothA {parent=Melee_TwoHanded_Handle, loc=<-9.1465 -5.5049 44.4427>, rot=<-0.374 -0.5647 0.476 -0.561>}
]
