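# Physics asset for the Freak one-handed axe melee weapon, giving its dangling tongue bone chain a setup for secondary physics motion.
# Small capsule bodies cover the handle bone and six dyn_tongue bones, generally tapering toward the tip.
# The tongue base joins the handle with free swing and twist, while each later segment links to the previous one with identical narrow angular limits.
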PhysicsAsset /BRCosmetics/Weapons/FORT_Melee/Freak_1h/Meshes/Freak_1h_Axe_Physics
bodies {
 Melee_TwoHanded_Handle [{shapeType=Capsule, length=1, radius=1, center=<0 -3.2766 23.5781>}]
 dyn_tongue_01_Base [{shapeType=Capsule, length=1, radius=1}]
 dyn_tongue_02 [{shapeType=Capsule, length=3.5, radius=1.5, center=<0.429 0 0>, rotation=<rot 90 0 0>}]
 dyn_tongue_03 [{shapeType=Capsule, length=3.25, radius=1, center=<0.4406 0 0>, rotation=<rot 90 0 0>}]
 dyn_tongue_04 [{shapeType=Capsule, length=1.5, radius=1, rotation=<rot 90 0 0>}]
 dyn_tongue_05 [{shapeType=Capsule, length=1.25, radius=0.75, rotation=<rot 90 -20 0>}]
 dyn_tongue_06 [{shapeType=Capsule, length=1.5, radius=0.5, rotation=<rot 90 -90 -79.084>}]
}
constraints [
 {
  bone1Name=dyn_tongue_01_Base
  bone2Name=Melee_TwoHanded_Handle
  swing1LimitDegrees=45
  swing1Motion=Free
  swing2LimitDegrees=45
  swing2Motion=Free
  twistLimitDegrees=45
  twistMotion=Free
 }
 {bone1Name=dyn_tongue_02, bone2Name=dyn_tongue_01_Base, swing1LimitDegrees=20, swing2LimitDegrees=25, twistLimitDegrees=10}
 {bone1Name=dyn_tongue_03, bone2Name=dyn_tongue_02, swing1LimitDegrees=20, swing2LimitDegrees=25, twistLimitDegrees=10}
 {bone1Name=dyn_tongue_04, bone2Name=dyn_tongue_03, swing1LimitDegrees=20, swing2LimitDegrees=25, twistLimitDegrees=10}
 {bone1Name=dyn_tongue_05, bone2Name=dyn_tongue_04, swing1LimitDegrees=20, swing2LimitDegrees=25, twistLimitDegrees=10}
 {bone1Name=dyn_tongue_06, bone2Name=dyn_tongue_05, swing1LimitDegrees=20, swing2LimitDegrees=25, twistLimitDegrees=10}
]
