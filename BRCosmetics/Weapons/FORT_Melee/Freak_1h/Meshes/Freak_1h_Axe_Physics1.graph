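# Physics asset for the Freak one-handed axe melee cosmetic, defining bodies that let the weapon's teeth move as secondary physics.
# Nine dyn_tooth bones each have a fitted capsule, while the Melee_TwoHanded_Handle bone has a small capsule as the shared root.
# Each tooth is constrained directly to the handle rather than chained, with identical moderate swing limits and a tight twist limit, so each can wobble independently.
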PhysicsAsset /BRCosmetics/Weapons/FORT_Melee/Freak_1h/Meshes/Freak_1h_Axe_Physics1
bodies {
 Melee_TwoHanded_Handle [{shapeType=Capsule, length=1, radius=1}]
 dyn_toothA [{shapeType=Capsule, length=3.2777, radius=1.3652, center=<2.1583 -0.0028 -0.3453>, rotation=<rot 0 93.5731 -86.9692>}]
 dyn_toothB [{shapeType=Capsule, length=4.782, radius=2.0498, center=<3.003 -0.2828 -0.463>, rotation=<rot 0 91.8886 -77.9913>}]
 dyn_toothC [{shapeType=Capsule, length=5.2536, radius=1.3232, center=<3.8341 -1.059 -0.1312>, rotation=<rot 0 101.5057 -89.6306>}]
 dyn_toothD [{shapeType=Capsule, length=8.3437, radius=2.1402, center=<6.528 -0.9946 -0.0704>, rotation=<rot 0 93.6993 -88.1983>}]
 dyn_toothE [{shapeType=Capsule, length=6.2325, radius=1.9875, center=<4.7411 -0.424 -0.0493>, rotation=<rot 0 98.5066 -84.7766>}]
 dyn_toothF [{shapeType=Capsule, length=3.6526, radius=1.8497, center=<1.8733 -0.7334 -0.0996>, rotation=<rot 0 102.1302 -83.0109>}]
 dyn_toothG [{shapeType=Capsule, length=6.0628, radius=1.8012, center=<3.4525 -1.4726 -0.1277>, rotation=<rot 0 -80.2125 -88.7745>}]
 dyn_toothH [{shapeType=Capsule, length=4.0875, radius=2.959, center=<2.2033 -0.158 0.0027>, rotation=<rot 0 -87.817 -84.6619>}]
 dyn_toothJ [{shapeType=Capsule, length=3.7192, radius=2.0554, center=<2.9141 -0.0369 0.0221>, rotation=<rot 0 90.6217 -86.7831>}]
}
constraints [
 {bone1Name=dyn_toothA, bone2Name=Melee_TwoHanded_Handle, swing1LimitDegrees=25, swing2LimitDegrees=25, twistLimitDegrees=5}
 {bone1Name=dyn_toothB, bone2Name=Melee_TwoHanded_Handle, swing1LimitDegrees=25, swing2LimitDegrees=25, twistLimitDegrees=5}
 {bone1Name=dyn_toothC, bone2Name=Melee_TwoHanded_Handle, swing1LimitDegrees=25, swing2LimitDegrees=25, twistLimitDegrees=5}
 {bone1Name=dyn_toothD, bone2Name=Melee_TwoHanded_Handle, swing1LimitDegrees=25, swing2LimitDegrees=25, twistLimitDegrees=5}
 {bone1Name=dyn_toothE, bone2Name=Melee_TwoHanded_Handle, swing1LimitDegrees=25, swing2LimitDegrees=25, twistLimitDegrees=5}
 {bone1Name=dyn_toothF, bone2Name=Melee_TwoHanded_Handle, swing1LimitDegrees=25, swing2LimitDegrees=25, twistLimitDegrees=5}
 {bone1Name=dyn_toothG, bone2Name=Melee_TwoHanded_Handle, swing1LimitDegrees=25, swing2LimitDegrees=25, twistLimitDegrees=5}
 {bone1Name=dyn_toothH, bone2Name=Melee_TwoHanded_Handle, swing1LimitDegrees=25, swing2LimitDegrees=25, twistLimitDegrees=5}
 {bone1Name=dyn_toothJ, bone2Name=Melee_TwoHanded_Handle, swing1LimitDegrees=25, swing2LimitDegrees=25, twistLimitDegrees=5}
]
