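# Animation Blueprint for the Freak_1h axe cosmetic melee weapon, adding physics and randomly shifting eyes.
# A looping eye animation, start-offset between main and offhand meshes by component name, feeds two RigidBody passes with different gravity, then additive eye-bone rotations and a Control Rig.
# Every 1–3 seconds, the update event randomly picks forward or backward eye rotation at random strength, smoothed by alpha interpolation.
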AnimBlueprint /BRCosmetics/Weapons/FORT_Melee/Freak_1h/Meshes/Freak_1h_Axe_AnimBP
target_skeleton=&/Freak_1h_Axe_Skeleton
graphs [
 AnimGraph {nodes [
   Input Pose()
   AnimGraphNode_StateMachine_0 = New State Machine() {is=title AnimGraphNode_StateMachine}
   n1 = RigidBody(ComponentPose=Local To Component(LocalPose=@AnimGraphNode_StateMachine_0)) {
    bClampLinearTranslationLimitToRefPose=true
    bOverrideWorldGravity=true
    overridePhysicsAsset=&/Freak_1h_Axe_Physics
    simulationSpace=BaseBoneSpace
    alphaBoolBlend {blendOption=Linear}
    baseBoneRef {boneName=b_AR_Root}
    componentLinearAccScale=<0.25 0.5 0.55>
    componentLinearVelScale=<0.25 0.5 0.55>
    overrideWorldGravity {z=-980}
   }
   n2 = RigidBody(ComponentPose=@n1) {
    bClampLinearTranslationLimitToRefPose=true
    bOverrideWorldGravity=true
    overridePhysicsAsset=&/Freak_1h_Axe_Physics1
    simulationSpace=BaseBoneSpace
    alphaBoolBlend {blendOption=Linear}
    baseBoneRef {boneName=b_AR_Root}
    componentAppliedLinearAccClamp {z=500}
    componentLinearAccScale=<1 1 0.35>
    componentLinearVelScale=<0.7 0.7 0.34>
    overrideWorldGravity {z=-150}
   }
   n3 = Get eye_fwd() -> (eye_fwd:double)
   n4 = "Transform (Modify) Bone - Bone: dyn_eye_l"(ComponentPose=@n2, Alpha=@n3, Translation="0,0,0", Rotation="45,0.000,0.000", Scale="1,1,1") {
    rotationMode=BMM_Additive
    rotationSpace=BCS_BoneSpace
    alphaBoolBlend {blendOption=Linear}
    alphaScaleBiasClamp {bInterpResult=true, interpSpeedDecreasing=15, interpSpeedIncreasing=15}
    boneToModify {boneName=dyn_eye_l}
    rotation {pitch=45}
   }
   n5 = "Transform (Modify) Bone - Bone: dyn_eye_r"(ComponentPose=@n4, Alpha=@n3, Translation="0,0,0", Rotation="-45,0.000,0.000", Scale="1,1,1") {
    rotationMode=BMM_Additive
    rotationSpace=BCS_BoneSpace
    alphaBoolBlend {blendOption=Linear}
    alphaScaleBiasClamp {bInterpResult=true, interpSpeedDecreasing=15, interpSpeedIncreasing=15}
    boneToModify {boneName=dyn_eye_r}
    rotation {pitch=-45}
   }
   n6 = Get eye_bck() -> (eye_bck:double)
   n7 = "Transform (Modify) Bone - Bone: dyn_eye_l"(ComponentPose=@n5, Alpha=@n6, Translation="0,0,0", Rotation="-45,0,0", Scale="1,1,1") {
    rotationMode=BMM_Additive
    rotationSpace=BCS_BoneSpace
    alphaBoolBlend {blendOption=Linear}
    alphaScaleBiasClamp {bInterpResult=true, interpSpeedDecreasing=15, interpSpeedIncreasing=15}
    boneToModify {boneName=dyn_eye_l}
    rotation {pitch=-45}
   }
   n8 = "Transform (Modify) Bone - Bone: dyn_eye_r"(ComponentPose=@n7, Alpha=@n6, Translation="0,0,0", Rotation="45,0,0", Scale="1,1,1") {
    rotationMode=BMM_Additive
    rotationSpace=BCS_BoneSpace
    alphaBoolBlend {blendOption=Linear}
    alphaScaleBiasClamp {bInterpResult=true, interpSpeedDecreasing=15, interpSpeedIncreasing=15}
    boneToModify {boneName=dyn_eye_r}
    rotation {pitch=45}
   }
   n9 = Control Rig(Source=Component To Local(ComponentPose=@n8)) {
    alphaBoolBlend {blendOption=Linear}
    controlRigAssetReference {blueprintRigClass=&/Freak_1h_Axe_CtrlRig.Freak_1h_Axe_CtrlRig_C}
   }
   Output Pose(Result=@n9) {layerGroup=None}
  ]}
 New State Machine {
  entry=BaseIdle
  path=AnimGraph.AnimGraphNode_StateMachine_0.New State Machine
  ids=id
  nodes [AnimStateNode_0 = BaseIdle() {is=title AnimStateNode}]
 }
 BaseIdle {
  path=AnimGraph.AnimGraphNode_StateMachine_0.New State Machine.AnimStateNode_0.BaseIdle
  nodes [Output Animation Pose(Result=Blend Poses by bool(
     BlendPose_0=Sequence Player 'Freak_1h_Axe_Eyes'() {sequence=&/Freak_1h_Axe_Eyes}
     BlendPose_1=Sequence Player 'Freak_1h_Axe_Eyes'() {sequence=&Freak_1h_Axe_Eyes, startPosition=1}
     bActiveValue=Get offset_weapon() -> (offset_weapon:bool)
    ) {blendType=Linear})]
 }
 EventGraph {nodes [
   Event BlueprintUpdateAnimation()
   > Sequence()
   > Delay(Duration=RandomFloatInRange(Min=1, Max=3))
   > n1 = Branch(Condition=RandomBool!() -> (ReturnValue:bool))
   n2 = RandomFloatInRange(Min=0.1, Max=0.8)
   > Set eye_fwd(eye_fwd:double=@n2) -> (Output_Get:double)
   > Set eye_bck(eye_bck:double=0) -> (Output_Get:double)
   Set eye_fwd(execute=@n1.else, eye_fwd=0)
   > Set eye_bck(eye_bck=@n2)
   Event BlueprintInitializeAnimation()
   > n3 = Set Mesh_Name(Mesh_Name:Skeletal Mesh Component Object Reference=GetPartSkeletalMeshComponent()) -> (Output_Get:Skeletal Mesh Component Object Reference)
   > n4 = Branch(Condition="Equal Exactly (String)"(A="To String (Object)"!(InObj:Object Reference=@n3) -> (ReturnValue:String), B=WeaponMeshOffhand))
   > Set offset_weapon(offset_weapon:bool=true) -> (Output_Get:bool)
   Set offset_weapon(execute=@n4.else, offset_weapon=false)
  ]}
]
variables [
 eye_bck {type=number}
 eye_fwd {type=number}
 Mesh_Name {type=SkeletalMeshComponent}
 offset_weapon {type=boolean}
]
